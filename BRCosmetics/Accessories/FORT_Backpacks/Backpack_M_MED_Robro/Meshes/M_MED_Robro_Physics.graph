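# Physics asset for the Robro backpack cosmetic, providing simple collision and joints for the pack and its dangling parts.
# A box body covers the pack bone, and small capsules cover the dyn_cords and dyn_handle bones.
# Both small bodies are constrained to the pack, with swing and twist limits set on the cords.
# The handle joint locks both swing axes, so it rotates only around twist, like a hinge.
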PhysicsAsset /BRCosmetics/Accessories/FORT_Backpacks/Backpack_M_MED_Robro/Meshes/M_MED_Robro_Physics
bodies {
 dyn_cords [{shapeType=Capsule, length=2, radius=1, center=<5 0 0.364>, rotation=<rot 0 0 90>}]
 dyn_handle [{shapeType=Capsule, length=2.7778, radius=0.8859, center=<0 0 2.1659>}]
 pack [{shapeType=Box, extentX=34.5605, extentY=7.9094, extentZ=17.2707, center=<-0.0001 -3.7229 -1.4032>}]
}
constraints [
 {bone1Name=dyn_cords, bone2Name=pack, swing1LimitDegrees=45, swing2LimitDegrees=45, twistLimitDegrees=35}
 {
  bone1Name=dyn_handle
  bone2Name=pack
  swing1LimitDegrees=45
  swing1Motion=Locked
  swing2LimitDegrees=45
  swing2Motion=Locked
  twistLimitDegrees=50
 }
]
